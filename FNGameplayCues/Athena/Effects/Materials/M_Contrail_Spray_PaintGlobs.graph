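# Translucent lit particle material for paint globs in a spray contrail, set up for Niagara and particle sprites.
# A random sub-UV function picks one cell of a 2x2 blurred-chunk atlas per particle, feeding both diffuse and normal samples.
# The chunk normal is angle-corrected blended with a slowly panning poison tile normal for moving surface detail.
# Particle color RGB is clamped, and lighting uses volumetric per-vertex directional translucency.
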Material /FNGameplayCues/Athena/Effects/Materials/M_Contrail_Spray_PaintGlobs
graph {
 properties {
  bAutomaticallySetUsageInEditor=false
  bUsedWithNiagaraSprites=true
  bUsedWithParticleSprites=true
  bUsedWithStaticMesh=false
  blendMode=BLEND_Translucent
  ditherOpacityMask=true
  pixelDepthOffsetMode=PDOM_Legacy
  translucencyLightingMode=TLM_VolumetricPerVertexDirectional
  usedShadingModels=Default Lit
 }
 nodes [
  Clamp(None=ParticleColor().RGB)
  n1 = Constant() {r=2}
  n2 = MaterialFunctionCall!(UseTexCoord 1, Rows=@n1, Columns=@n1, Particle Random Value=ParticleRandom()) {materialFunction=/Game/Effects/Fort_Effects/Materials/_Functions/MF_RandomSubUV_GPU_Sprites}
  MaterialFunctionCall(
   BaseNormal=TextureSample(UVs=@n2.Result) {samplerType=SAMPLERTYPE_Normal, texture=/Game/Effects/Fort_Effects/Textures/Rocks/T_ChunksBlur_2x2_N}.RGB
   AdditionalNormal=TextureSample(UVs=Panner() {speedY=-0.1}) {samplerType=SAMPLERTYPE_Normal, texture=/Game/Effects/Fort_Effects/Textures/Husk/Poison/T_PoisonTile_N}.RGB
  ) {materialFunction=&BlendAngleCorrectedNormals}
  TextureSample(UVs=@n2.Result) {texture=/Game/Effects/Fort_Effects/Textures/Rocks/T_ChunksBlur_2x2_D}
 ]
}
