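# Unlit translucent particle sprite material for sparkles in the Inky glider deploy effect.
# Particle random blends among two sparkle textures and a starburst, tinted by sine-twinkled, randomly hue-shifted particle color.
# Mobile instead multiplies starburst, particle color and a hue-shifted Fresnel rainbow thin-film, feeding emissive adjustments through a shading path switch.
# The thin-film roughness and visibility lerp has identical inputs, so that mask has no effect.
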Material /FNGameplayCues/Athena/Effects/Materials/M_GliderDeploy_Sparkles_Inky
graph {
 properties {
  bUsedWithParticleSprites=true
  blendMode=BLEND_Translucent
  pixelDepthOffsetMode=PDOM_Legacy
  shadingModel=MSM_Unlit
  usedShadingModels=Unlit
 }
 nodes [
  n1 = TextureSample() {texture=/Game/Effects/Fort_Effects/Textures/Flares/T_Starburst}
  n2 = ParticleRandom()
  n3 = MaterialFunctionCall(
   A=TextureSample() {texture=/Game/Effects/Fort_Effects/Textures/Sparkle/T_Sparkle_01}.R
   B=TextureSample() {texture=/Game/CharClasses/Outlander/Gadgets/AntiMaterialCharge/Materials/T_AntiMaterialCharge_Sparks_01}.G
   C=@n1.G
   Alpha=Ceil(None=Multiply(A=@n2) {constB=3})
  ) {materialFunction=&3ColorBlend}
  n4 = ParticleColor()
  n5 = ParticleRandom()
  n6 = Multiply(A=@n5) {constB=0.15}
  n7 = MaterialFunctionCall(Speed=Multiply(A=@n5, B=ScalarParameter() {defaultValue=3, parameterName=TwinkleSpeed})) {materialFunction=&DebugTimeSine}
  n8 = Multiply(A=Clamp(None=Multiply(A=@n7.Result) {constB=5}) {maxDefault=5, minDefault=0.25}, B=@n4.RGB)
  n9 = Multiply(
   A=@n3.Result
   B=LinearInterpolate(
    A=@n4.RGB
    B=MaterialFunctionCall(Hue Shift Percentage=@n6, Texture=@n8) {materialFunction=&HueShift}.Result
    Alpha=ScalarParameter() {defaultValue=1, parameterName=Use Particle Random Color}
   )
  )
  n10 = ScalarParameter() {defaultValue=1, group=ThinFilm, parameterName=ThinFilm_Scale}
  n11 = ParticleDirection()
  n12 = Fresnel(ExponentIn=ScalarParameter() {defaultValue=1, group=ThinFilm, parameterName=ThinFilm_EX}, Normal=@n11) {baseReflectFraction=0}
  n13 = Multiply(
   A=TextureSample(UVs=Multiply(A=@n10, B=@n12)) {texture=/Game/Global/Textures/Default/Masks_and_Gradients/FortniteRainbow}.RGB
   B=ScalarParameter() {defaultValue=1, group=ThinFilm, parameterName=ThinFilm_Intensity}
  )
  n14 = ScalarParameter() {defaultValue=0.7, parameterName=Roughness}
  n15 = Clamp(None=Multiply(A=OneMinus(None=@n14), B=ScalarParameter() {defaultValue=1, group=ThinFilm, parameterName=ThinFilm_RoughnessScale}))
  n16 = Fresnel(ExponentIn=ScalarParameter() {defaultValue=1, group=ThinFilm, parameterName=ThinFilm_Visibility}, Normal=@n11) {baseReflectFraction=0}
  n17 = MaterialFunctionCall(
   Hue Shift Percentage=ScalarParameter() {defaultValue=0.1, parameterName="Hue Shift Percentage (S)"}
   Texture=LinearInterpolate(A=@n13, B=@n13, Alpha=Multiply(A=@n15, B=@n16))
  ) {materialFunction=&HueShift}
  n18 = Multiply(A=@n4.RGB, B=@n1.G)
  MaterialFunctionCall(Emissive=ShadingPathSwitch(Default=@n9, Mobile=Multiply(A=@n17.Result, B=@n18)) {desc=Migrated from FeatureLevelSwitch}) {materialFunction=&MF_FX_EmissiveAdjustments}
  Multiply(A=@n3.Result, B=ParticleColor().A)
  Multiply(A=@n17.Result, B=@n9)
  VertexNormalWS()
 ]
}
